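# Physics asset for the DangerMojo backpack cosmetic, defining simulation shapes for two of its dyn-prefixed joints.
# A large capsule covers the base joint and a smaller, thinner capsule covers the first stripe joint.
# A single constraint links the stripe to the base, with asymmetric swing limits and a twist limit of zero.
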PhysicsAsset /BRCosmetics/Accessories/FORT_Backpacks/Backpack_F_MED_DangerMojo/Meshes/F_MED_DangerMojo_Pack_Physics
bodies {
 dyn_base_joint [{
   shapeType=Capsule
   length=15.455
   radius=2.6382
   shapeName=dyn_base_joint_capsule
   center=<7.0956 -0.0485 -6.3815>
   rotation=<rot 0 -90.3232 -49.3422>
  }]
 dyn_stripe_1_joint [{
   shapeType=Capsule
   length=4.7313
   radius=0.8924
   shapeName=dyn_stripe_1_joint_capsule
   center=<12.8972 -0.0005 0.0041>
   rotation=<rot 0 -90.0074 -89.6391>
  }]
}
constraints [{bone1Name=dyn_stripe_1_joint, bone2Name=dyn_base_joint, swing1LimitDegrees=10, swing2LimitDegrees=30, twistLimitDegrees=0}]
