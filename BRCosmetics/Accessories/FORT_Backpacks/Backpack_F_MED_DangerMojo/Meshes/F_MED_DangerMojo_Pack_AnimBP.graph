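# Animation Blueprint for the DangerMojo cosmetic backpack, adding physics sway and a skydiving-dependent offset.
# A RigidBody node simulates the pack with its own physics asset, using scaled and clamped component motion.
# Its output feeds two additive Modify Bone nodes on dyn_base_joint, keyed to opposite states of bIsSkyDiving with sinusoidal blends.
# The event graph holds only unconnected stub nodes, so nothing here sets that flag.
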AnimBlueprint /BRCosmetics/Accessories/FORT_Backpacks/Backpack_F_MED_DangerMojo/Meshes/F_MED_DangerMojo_Pack_AnimBP
target_skeleton=&/F_MED_DangerMojo_Pack_Skeleton
graphs [
 AnimGraph {nodes [
   n1 = RigidBody(ComponentPose=Local To Component(LocalPose=Input Pose())) {
    overridePhysicsAsset=&/F_MED_DangerMojo_Pack_Physics
    alphaBoolBlend {blendOption=Linear}
    componentAppliedLinearAccClamp=<2000 2000 2000>
    componentLinearAccScale=<0.6 0.6 0.6>
    componentLinearVelScale=<0.6 0.6 0.6>
    simSpaceSettings {maxAngularAcceleration=2000, maxAngularVelocity=2000, maxLinearAcceleration=2000, maxLinearVelocity=2000, velocityScaleZ=0.8, worldAlpha=0.1}
   }
   n2 = "Transform (Modify) Bone - Bone: dyn_base_joint"!(ComponentPose=@n1, bAlphaBoolEnabled:bool=NOT Boolean(A=Get bIsSkyDiving()), Translation="-7,-1,7", Rotation="10,0,0", Scale="1,1,1") -> (Pose) {
    alphaInputType=Bool
    rotationMode=BMM_Additive
    translationMode=BMM_Additive
    alphaBoolBlend {blendInTime=1, blendOption=Sinusoidal, blendOutTime=1}
    boneToModify {boneName=dyn_base_joint}
    rotation {pitch=10}
    translation=<-7 -1 7>
   }
   n3 = "Transform (Modify) Bone - Bone: dyn_base_joint"(ComponentPose=@n2, bAlphaBoolEnabled=Get bIsSkyDiving(), Translation="0,-6,9", Rotation="-10,-25,10", Scale="1,1,1") {
    alphaInputType=Bool
    rotationMode=BMM_Additive
    translationMode=BMM_Additive
    alphaBoolBlend {blendInTime=0.1, blendOption=Sinusoidal, blendOutTime=0.1}
    boneToModify {boneName=dyn_base_joint}
    rotation=<rot -10 -25 10>
    translation {y=-6, z=9}
   }
   Output Pose(Result=Component To Local(ComponentPose=@n3)) {layerGroup=None}
  ]}
 EventGraph {nodes [
   TryGetPawnOwner()
   Event BlueprintUpdateAnimation()
  ]}
]
variables [Is Sky Diving {type=boolean}]
